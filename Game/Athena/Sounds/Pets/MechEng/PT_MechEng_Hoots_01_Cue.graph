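# Sound cue for the mechanical owl hoot of the MechEng pet.
# A random node picks, with equal weight, one of five wave players: four single-hoot variations and one double hoot.
# It plays through the creature pet sound class with close-range pet attenuation at reduced volume, and is excluded from random-branch culling.
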SoundCue /Game/Athena/Sounds/Pets/MechEng/PT_MechEng_Hoots_01_Cue
root=Random_2
nodes [
 {id=Random_2, children [WavePlayer_5,WavePlayer_7,WavePlayer_8,WavePlayer_9,WavePlayer_0], properties {weights [1,1,1,1,1]}}
 {id=WavePlayer_0, properties {soundWaveAssetPtr=&/Pet_MechOwl_2Hoots_01}}
 {id=WavePlayer_5, properties {soundWaveAssetPtr=&/Pet_MechOwl_1Hoot_01}}
 {id=WavePlayer_7, properties {soundWaveAssetPtr=&/Pet_MechOwl_1Hoot_03}}
 {id=WavePlayer_8, properties {soundWaveAssetPtr=&/Pet_MechOwl_1Hoot_04}}
 {id=WavePlayer_9, properties {soundWaveAssetPtr=&/Pet_MechOwl_1Hoot_05}}
]
properties {
 attenuationSettings=&Fort_Pets_Close_Athena_Att
 bExcludeFromRandomNodeBranchCulling=true
 maxDistance=1800.0001
 soundClassObject=&Fort_Creatures_Pets
 volumeMultiplier=0.672
}
