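# Sound cue for the MechEng mech owl pet's transition from closed to glide, judging by its name.
# A single wave player feeds the output directly, with no randomization, mixing or modulation nodes.
# It routes through the creature pets sound class and borrows the player glider open/close attenuation for spatial falloff, with a slightly reduced volume.
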SoundCue /Game/Athena/Sounds/Pets/MechEng/Pet_MechOwl_ClosedToGlide_Cue
root=WavePlayer_0
nodes [{id=WavePlayer_0, properties {soundWaveAssetPtr=&/Pet_MechOwl_ClosedToGlide}}]
properties {
 attenuationSettings=/Game/Sounds/Attenuation/Fort_Player_Glider_OpenClose_Athena_Att
 bExcludeFromRandomNodeBranchCulling=true
 maxDistance=10000
 soundClassObject=&Fort_Creatures_Pets
 volumeMultiplier=0.896
}
