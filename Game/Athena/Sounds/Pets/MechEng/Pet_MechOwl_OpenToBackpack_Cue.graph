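# Sound cue for the MechOwl pet's open-to-backpack moment in Athena.
# A single wave player feeds the GlideToBackpack wave straight to the output, with no randomization or modulation nodes.
# It uses the close-range pet attenuation and the creature pets sound class, with slightly reduced volume.
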SoundCue /Game/Athena/Sounds/Pets/MechEng/Pet_MechOwl_OpenToBackpack_Cue
root=WavePlayer_1
nodes [{id=WavePlayer_1, properties {soundWaveAssetPtr=&/Pet_MechOwl_GlideToBackpack}}]
properties {
 attenuationSettings=&Fort_Pets_Close_Athena_Att
 bExcludeFromRandomNodeBranchCulling=true
 maxDistance=1800.0001
 soundClassObject=&Fort_Creatures_Pets
 volumeMultiplier=0.896
}
